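# Physics asset for the Phantom medium male player's coat, simulating its tails with pelvis, leg and foot bodies as colliders.
# Seven five-segment chains (three per side, one back) hang from base spheres jointed to root, with limited or locked twist.
# Angular-free cross constraints link neighboring chains at the lower three rows, so the coat moves as a sheet.
# A huge root capsule below the origin apparently serves as a floor.
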PhysicsAsset /Game/Characters/Player/Male/Medium/Bodies/M_MED_Phantom/Meshes/M_MED_Phantom_Coat_Physics
bodies {
 calf_l [{shapeType=Capsule, length=42.7133, radius=8.3857, center=<-23.9185 -0.9477 -0.3689>, rotation=<rot -0.0007 90.0094 -89.9897>}]
 calf_r [{shapeType=Capsule, length=42.7133, radius=8.3857, center=<23.9185 0.9477 0.3689>, rotation=<rot -0.0007 90.0094 90.0103>}]
 dyn_coat_a_A_base_l [{shapeType=Sphere, radius=1}]
 dyn_coat_a_A_base_r [{shapeType=Sphere, radius=1}]
 dyn_coat_a_A_l [{shapeType=Box, extentX=13, extentY=12.5, extentZ=2.5, center=<6.4233 -2.2251 -0.5545>, rotation=<rot -0.2556 5.1012 -12.8729>}]
 dyn_coat_a_A_r [{shapeType=Box, extentX=13, extentY=19.5, extentZ=2.5, center=<-6.4233 2.2251 0.5545>, rotation=<rot -0.2556 5.1013 166.7086>}]
 dyn_coat_a_B_l [{shapeType=Box, extentX=13, extentY=20.5, extentZ=3.5, center=<4.8636 -1.6236 -0.5409>, rotation=<rot -3.7457 4.7634 -18.079>}]
 dyn_coat_a_B_r [{shapeType=Box, extentX=13, extentY=22.5, extentZ=3.5, center=<-4.8636 1.6236 0.5409>, rotation=<rot 0 0 163.2861>}]
 dyn_coat_a_C_l [{shapeType=Box, extentX=14, extentY=20.5, extentZ=4, center=<4.8209 -1.8279 -0.0309>, rotation=<rot 0 0 -8.0297>}]
 dyn_coat_a_C_r [{shapeType=Box, extentX=14, extentY=22.5, extentZ=4, center=<-4.8209 1.8279 0.0309>, rotation=<rot 0 0 173.3671>}]
 dyn_coat_a_D_l [{shapeType=Box, extentX=14, extentY=20.5, extentZ=4, center=<4.7858 -1.897 0.2851>, rotation=<rot 0 0 -8.0297>}]
 dyn_coat_a_D_r [{shapeType=Box, extentX=14, extentY=22.5, extentZ=4, center=<-4.7858 1.897 -0.2851>, rotation=<rot 0 0 173.3671>}]
 dyn_coat_a_E_l [{shapeType=Capsule, length=15, radius=7, shapeName=dyn_coat_a_E_l_capsule, center=<3 0 0>, rotation=<rot 0 0 90>}]
 dyn_coat_a_E_r [{shapeType=Capsule, length=15, radius=7, shapeName=dyn_coat_a_E_r_capsule, center=<-3 0 0>, rotation=<rot 0 0 90>}]
 dyn_coat_b_A_base_l [{shapeType=Sphere, radius=1}]
 dyn_coat_b_A_base_r [{shapeType=Sphere, radius=1}]
 dyn_coat_b_A_l [{shapeType=Box, extentX=13, extentY=12, extentZ=2.5, center=<6.4517 -0.0729 -0.4135>}]
 dyn_coat_b_A_r [{shapeType=Box, extentX=13, extentY=16, extentZ=2.5, center=<-6.4517 0.0729 0.4135>, rotation=<rot 0 0 180>}]
 dyn_coat_b_B_l [{shapeType=Capsule, length=15, radius=5, shapeName=dyn_coat_b_B_l_capsule, center=<3 0 0>, rotation=<rot 0 0 90>}]
 dyn_coat_b_B_r [{shapeType=Capsule, length=15, radius=5, shapeName=dyn_coat_b_B_r_capsule, center=<-3 0 0>, rotation=<rot 0 0 90>}]
 dyn_coat_b_C_l [{shapeType=Capsule, length=15, radius=5, shapeName=dyn_coat_b_C_l_capsule, center=<3 0 0>, rotation=<rot 0 0 90>}]
 dyn_coat_b_C_r [{shapeType=Capsule, length=15, radius=5, shapeName=dyn_coat_b_C_r_capsule, center=<-3 0 0>, rotation=<rot 0 0 90>}]
 dyn_coat_b_D_l [{shapeType=Capsule, length=15, radius=7, shapeName=dyn_coat_b_D_l_capsule, center=<3 0 0>, rotation=<rot 0 0 90>}]
 dyn_coat_b_D_r [{shapeType=Capsule, length=15, radius=7, shapeName=dyn_coat_b_D_r_capsule, center=<-3 0 0>, rotation=<rot 0 0 90>}]
 dyn_coat_b_E_l [{shapeType=Capsule, length=15, radius=8, shapeName=dyn_coat_b_E_l_capsule, center=<3 0 0>, rotation=<rot 0 0 90>}]
 dyn_coat_b_E_r [{shapeType=Capsule, length=15, radius=8, shapeName=dyn_coat_b_E_r_capsule, center=<-3 0 0>, rotation=<rot 0 0 90>}]
 dyn_coat_bk_A [{shapeType=Box, extentX=11.5, extentY=4.6758, extentZ=26.9681, center=<6.1901 0.0245 0.1059>, rotation=<rot 0 179.3051 -87.5993>}]
 dyn_coat_bk_A_base [{shapeType=Sphere, radius=1}]
 dyn_coat_bk_B [{shapeType=Capsule, length=15, radius=7, shapeName=dyn_coat_bk_B_capsule, center=<3 0 0>, rotation=<rot 0 0 90>}]
 dyn_coat_bk_C [{shapeType=Capsule, length=15, radius=7, shapeName=dyn_coat_bk_C_capsule, center=<3 0 0>, rotation=<rot 0 0 90>}]
 dyn_coat_bk_D [{shapeType=Capsule, length=15, radius=7, shapeName=dyn_coat_bk_D_capsule, center=<3 0 0>, rotation=<rot 0 0 90>}]
 dyn_coat_bk_E [{shapeType=Capsule, length=15, radius=8, shapeName=dyn_coat_bk_E_capsule, center=<3 0 0>, rotation=<rot 0 0 90>}]
 dyn_coat_c_A_base_l [{shapeType=Sphere, radius=1}]
 dyn_coat_c_A_base_r [{shapeType=Sphere, radius=1}]
 dyn_coat_c_A_l [{shapeType=Box, extentX=13, extentY=15, extentZ=5, center=<7.4961 0 0.3195>}]
 dyn_coat_c_A_r [{shapeType=Box, extentX=13, extentY=15, extentZ=5, center=<-7.4961 0 -0.3195>, rotation=<rot 0 0 180>}]
 dyn_coat_c_B_l [{shapeType=Capsule, length=15, radius=7, shapeName=dyn_coat_c_B_l_capsule, center=<3 0 0>, rotation=<rot 0 0 90>}]
 dyn_coat_c_B_r [{shapeType=Capsule, length=15, radius=7, shapeName=dyn_coat_c_B_r_capsule, center=<-3 0 0>, rotation=<rot 0 0 90>}]
 dyn_coat_c_C_l [{shapeType=Capsule, length=15, radius=7, shapeName=dyn_coat_c_C_l_capsule, center=<3 0 0>, rotation=<rot 0 0 90>}]
 dyn_coat_c_C_r [{shapeType=Capsule, length=15, radius=7, shapeName=dyn_coat_c_C_r_capsule, center=<-3 0 0>, rotation=<rot 0 0 90>}]
 dyn_coat_c_D_l [{shapeType=Capsule, length=15, radius=7, shapeName=dyn_coat_c_D_l_capsule, center=<3 0 0>, rotation=<rot 0 0 90>}]
 dyn_coat_c_D_r [{shapeType=Capsule, length=15, radius=7, shapeName=dyn_coat_c_D_r_capsule, center=<-3 0 0>, rotation=<rot 0 0 90>}]
 dyn_coat_c_E_l [{shapeType=Capsule, length=15, radius=8, shapeName=dyn_coat_c_E_l_capsule, center=<3 0 0>, rotation=<rot 0 0 90>}]
 dyn_coat_c_E_r [{shapeType=Capsule, length=15, radius=8, shapeName=dyn_coat_c_E_r_capsule, center=<-3 0 0>, rotation=<rot 0 0 90>}]
 foot_l [{shapeType=Capsule, length=20, radius=9, shapeName=foot_l_capsule, center=<-2.6 6 0>, rotation=<rot 0 0 90>}]
 foot_r [{shapeType=Capsule, length=20, radius=9, shapeName=foot_l_capsule, center=<2.6 -6 0>, rotation=<rot 0 0 90>}]
 pelvis [{shapeType=Capsule, length=5.3538, radius=11.6282, center=<-6.4215 -0.5457 0.0725>}]
 root [{shapeType=Capsule, length=200, radius=300, shapeName=root_capsule, center=<0 0 -300.4243>, rotation=<rot 90 0 0>}]
 thigh_l [{shapeType=Capsule, length=27.1105, radius=8.1753, center=<-20.5193 0.7462 0.0288>, rotation=<rot -0.0222 -88.3553 -90.348>}]
 thigh_r [{shapeType=Capsule, length=27.1105, radius=8.1753, center=<20.5193 -0.7462 -0.0288>, rotation=<rot -0.0222 -88.3553 89.652>}]
}
constraints [
 {bone1Name=calf_l, bone2Name=thigh_l, swing1LimitDegrees=45, swing2LimitDegrees=45, twistLimitDegrees=45}
 {
  bone1Name=thigh_l
  bone2Name=pelvis
  swing1LimitDegrees=45
  swing1Motion=Free
  swing2LimitDegrees=45
  swing2Motion=Free
  twistLimitDegrees=45
  twistMotion=Free
 }
 {bone1Name=calf_r, bone2Name=thigh_r, swing1LimitDegrees=45, swing2LimitDegrees=45, twistLimitDegrees=45}
 {
  bone1Name=thigh_r
  bone2Name=pelvis
  swing1LimitDegrees=45
  swing1Motion=Free
  swing2LimitDegrees=45
  swing2Motion=Free
  twistLimitDegrees=45
  twistMotion=Free
 }
 {bone1Name=dyn_coat_a_A_l, bone2Name=dyn_coat_a_A_base_l, swing1LimitDegrees=15, swing2LimitDegrees=45, twistLimitDegrees=5}
 {bone1Name=dyn_coat_a_B_l, bone2Name=dyn_coat_a_A_l, swing1LimitDegrees=45, swing2LimitDegrees=45, twistLimitDegrees=5}
 {bone1Name=dyn_coat_a_C_l, bone2Name=dyn_coat_a_B_l, swing1LimitDegrees=45, swing2LimitDegrees=45, twistLimitDegrees=5}
 {bone1Name=dyn_coat_a_D_l, bone2Name=dyn_coat_a_C_l, swing1LimitDegrees=45, swing2LimitDegrees=45, twistLimitDegrees=5}
 {
  bone1Name=dyn_coat_a_E_l
  bone2Name=dyn_coat_a_D_l
  swing1LimitDegrees=45
  swing2LimitDegrees=45
  swing2Motion=Locked
  twistLimitDegrees=5
  twistMotion=Locked
 }
 {bone1Name=dyn_coat_bk_A, bone2Name=dyn_coat_bk_A_base, swing1LimitDegrees=45, swing2LimitDegrees=45, twistLimitDegrees=5}
 {bone1Name=dyn_coat_bk_B, bone2Name=dyn_coat_bk_A, swing1LimitDegrees=20, swing2LimitDegrees=45, twistLimitDegrees=5, twistMotion=Locked}
 {bone1Name=dyn_coat_bk_C, bone2Name=dyn_coat_bk_B, swing1LimitDegrees=20, swing2LimitDegrees=45, twistLimitDegrees=5, twistMotion=Locked}
 {bone1Name=dyn_coat_bk_D, bone2Name=dyn_coat_bk_C, swing1LimitDegrees=20, swing2LimitDegrees=45, twistLimitDegrees=5, twistMotion=Locked}
 {
  bone1Name=dyn_coat_bk_E
  bone2Name=dyn_coat_bk_D
  swing1LimitDegrees=20
  swing2LimitDegrees=45
  swing2Motion=Locked
  twistLimitDegrees=5
  twistMotion=Locked
 }
 {bone1Name=dyn_coat_b_A_l, bone2Name=dyn_coat_b_A_base_l, swing1LimitDegrees=45, swing2LimitDegrees=45, twistLimitDegrees=5}
 {bone1Name=dyn_coat_b_B_l, bone2Name=dyn_coat_b_A_l, swing1LimitDegrees=20, swing2LimitDegrees=45, twistLimitDegrees=5, twistMotion=Locked}
 {bone1Name=dyn_coat_b_C_l, bone2Name=dyn_coat_b_B_l, swing1LimitDegrees=20, swing2LimitDegrees=45, twistLimitDegrees=5, twistMotion=Locked}
 {bone1Name=dyn_coat_b_D_l, bone2Name=dyn_coat_b_C_l, swing1LimitDegrees=20, swing2LimitDegrees=45, twistLimitDegrees=5, twistMotion=Locked}
 {
  bone1Name=dyn_coat_b_E_l
  bone2Name=dyn_coat_b_D_l
  swing1LimitDegrees=20
  swing2LimitDegrees=45
  swing2Motion=Locked
  twistLimitDegrees=5
  twistMotion=Locked
 }
 {bone1Name=dyn_coat_c_A_l, bone2Name=dyn_coat_c_A_base_l, swing1LimitDegrees=45, swing2LimitDegrees=45, twistLimitDegrees=5}
 {bone1Name=dyn_coat_c_B_l, bone2Name=dyn_coat_c_A_l, swing1LimitDegrees=20, swing2LimitDegrees=45, twistLimitDegrees=5, twistMotion=Locked}
 {bone1Name=dyn_coat_c_C_l, bone2Name=dyn_coat_c_B_l, swing1LimitDegrees=20, swing2LimitDegrees=45, twistLimitDegrees=5, twistMotion=Locked}
 {bone1Name=dyn_coat_c_D_l, bone2Name=dyn_coat_c_C_l, swing1LimitDegrees=20, swing2LimitDegrees=45, twistLimitDegrees=5, twistMotion=Locked}
 {
  bone1Name=dyn_coat_c_E_l
  bone2Name=dyn_coat_c_D_l
  swing1LimitDegrees=20
  swing2LimitDegrees=45
  swing2Motion=Locked
  twistLimitDegrees=5
  twistMotion=Locked
 }
 {bone1Name=dyn_coat_a_B_r, bone2Name=dyn_coat_a_A_r, swing1LimitDegrees=45, swing2LimitDegrees=45, twistLimitDegrees=5}
 {bone1Name=dyn_coat_a_C_r, bone2Name=dyn_coat_a_B_r, swing1LimitDegrees=45, swing2LimitDegrees=45, twistLimitDegrees=5}
 {bone1Name=dyn_coat_a_D_r, bone2Name=dyn_coat_a_C_r, swing1LimitDegrees=45, swing2LimitDegrees=45, twistLimitDegrees=5}
 {
  bone1Name=dyn_coat_a_E_r
  bone2Name=dyn_coat_a_D_r
  swing1LimitDegrees=45
  swing2LimitDegrees=45
  swing2Motion=Locked
  twistLimitDegrees=5
  twistMotion=Locked
 }
 {bone1Name=dyn_coat_b_A_r, bone2Name=dyn_coat_b_A_base_r, swing1LimitDegrees=45, swing2LimitDegrees=45, twistLimitDegrees=5}
 {bone1Name=dyn_coat_b_B_r, bone2Name=dyn_coat_b_A_r, swing1LimitDegrees=20, swing2LimitDegrees=45, twistLimitDegrees=5, twistMotion=Locked}
 {bone1Name=dyn_coat_b_C_r, bone2Name=dyn_coat_b_B_r, swing1LimitDegrees=20, swing2LimitDegrees=45, twistLimitDegrees=5, twistMotion=Locked}
 {bone1Name=dyn_coat_b_D_r, bone2Name=dyn_coat_b_C_r, swing1LimitDegrees=20, swing2LimitDegrees=45, twistLimitDegrees=5, twistMotion=Locked}
 {
  bone1Name=dyn_coat_b_E_r
  bone2Name=dyn_coat_b_D_r
  swing1LimitDegrees=20
  swing2LimitDegrees=45
  swing2Motion=Locked
  twistLimitDegrees=5
  twistMotion=Locked
 }
 {bone1Name=dyn_coat_c_A_r, bone2Name=dyn_coat_c_A_base_r, swing1LimitDegrees=45, swing2LimitDegrees=45, twistLimitDegrees=5}
 {bone1Name=dyn_coat_c_B_r, bone2Name=dyn_coat_c_A_r, swing1LimitDegrees=20, swing2LimitDegrees=45, twistLimitDegrees=5, twistMotion=Locked}
 {bone1Name=dyn_coat_c_C_r, bone2Name=dyn_coat_c_B_r, swing1LimitDegrees=20, swing2LimitDegrees=45, twistLimitDegrees=5, twistMotion=Locked}
 {bone1Name=dyn_coat_c_D_r, bone2Name=dyn_coat_c_C_r, swing1LimitDegrees=20, swing2LimitDegrees=45, twistLimitDegrees=5, twistMotion=Locked}
 {
  bone1Name=dyn_coat_c_E_r
  bone2Name=dyn_coat_c_D_r
  swing1LimitDegrees=20
  swing2LimitDegrees=45
  swing2Motion=Locked
  twistLimitDegrees=5
  twistMotion=Locked
 }
 {bone1Name=dyn_coat_a_A_r, bone2Name=dyn_coat_a_A_base_r, swing1LimitDegrees=45, swing2LimitDegrees=45, twistLimitDegrees=5}
 {
  bone1Name=dyn_coat_b_C_l
  bone2Name=dyn_coat_a_C_l
  swing1LimitDegrees=45
  swing1Motion=Free
  swing2LimitDegrees=45
  swing2Motion=Free
  twistLimitDegrees=45
  twistMotion=Free
 }
 {
  bone1Name=dyn_coat_c_C_l
  bone2Name=dyn_coat_b_C_l
  swing1LimitDegrees=45
  swing1Motion=Free
  swing2LimitDegrees=45
  swing2Motion=Free
  twistLimitDegrees=45
  twistMotion=Free
 }
 {
  bone1Name=dyn_coat_bk_C
  bone2Name=dyn_coat_c_C_l
  swing1LimitDegrees=45
  swing1Motion=Free
  swing2LimitDegrees=45
  swing2Motion=Free
  twistLimitDegrees=45
  twistMotion=Free
 }
 {
  bone1Name=dyn_coat_b_C_r
  bone2Name=dyn_coat_a_C_r
  swing1LimitDegrees=45
  swing1Motion=Free
  swing2LimitDegrees=45
  swing2Motion=Free
  twistLimitDegrees=45
  twistMotion=Free
 }
 {
  bone1Name=dyn_coat_c_C_r
  bone2Name=dyn_coat_b_C_r
  swing1LimitDegrees=45
  swing1Motion=Free
  swing2LimitDegrees=45
  swing2Motion=Free
  twistLimitDegrees=45
  twistMotion=Free
 }
 {
  bone1Name=dyn_coat_bk_C
  bone2Name=dyn_coat_c_C_r
  swing1LimitDegrees=45
  swing1Motion=Free
  swing2LimitDegrees=45
  swing2Motion=Free
  twistLimitDegrees=45
  twistMotion=Free
 }
 {
  bone1Name=dyn_coat_b_D_l
  bone2Name=dyn_coat_a_D_l
  swing1LimitDegrees=45
  swing1Motion=Free
  swing2LimitDegrees=45
  swing2Motion=Free
  twistLimitDegrees=45
  twistMotion=Free
 }
 {
  bone1Name=dyn_coat_c_D_l
  bone2Name=dyn_coat_b_D_l
  swing1LimitDegrees=45
  swing1Motion=Free
  swing2LimitDegrees=45
  swing2Motion=Free
  twistLimitDegrees=45
  twistMotion=Free
 }
 {
  bone1Name=dyn_coat_bk_D
  bone2Name=dyn_coat_c_D_l
  swing1LimitDegrees=45
  swing1Motion=Free
  swing2LimitDegrees=45
  swing2Motion=Free
  twistLimitDegrees=45
  twistMotion=Free
 }
 {
  bone1Name=dyn_coat_b_D_r
  bone2Name=dyn_coat_a_D_r
  swing1LimitDegrees=45
  swing1Motion=Free
  swing2LimitDegrees=45
  swing2Motion=Free
  twistLimitDegrees=45
  twistMotion=Free
 }
 {
  bone1Name=dyn_coat_c_D_r
  bone2Name=dyn_coat_b_D_r
  swing1LimitDegrees=45
  swing1Motion=Free
  swing2LimitDegrees=45
  swing2Motion=Free
  twistLimitDegrees=45
  twistMotion=Free
 }
 {
  bone1Name=dyn_coat_bk_D
  bone2Name=dyn_coat_c_D_r
  swing1LimitDegrees=45
  swing1Motion=Free
  swing2LimitDegrees=45
  swing2Motion=Free
  twistLimitDegrees=45
  twistMotion=Free
 }
 {
  bone1Name=dyn_coat_b_E_l
  bone2Name=dyn_coat_a_E_l
  swing1LimitDegrees=45
  swing1Motion=Free
  swing2LimitDegrees=45
  swing2Motion=Free
  twistLimitDegrees=45
  twistMotion=Free
 }
 {
  bone1Name=dyn_coat_c_E_l
  bone2Name=dyn_coat_b_E_l
  swing1LimitDegrees=45
  swing1Motion=Free
  swing2LimitDegrees=45
  swing2Motion=Free
  twistLimitDegrees=45
  twistMotion=Free
 }
 {
  bone1Name=dyn_coat_bk_E
  bone2Name=dyn_coat_c_E_l
  swing1LimitDegrees=45
  swing1Motion=Free
  swing2LimitDegrees=45
  swing2Motion=Free
  twistLimitDegrees=45
  twistMotion=Free
 }
 {
  bone1Name=dyn_coat_b_E_r
  bone2Name=dyn_coat_a_E_r
  swing1LimitDegrees=45
  swing1Motion=Free
  swing2LimitDegrees=45
  swing2Motion=Free
  twistLimitDegrees=45
  twistMotion=Free
 }
 {
  bone1Name=dyn_coat_c_E_r
  bone2Name=dyn_coat_b_E_r
  swing1LimitDegrees=45
  swing1Motion=Free
  swing2LimitDegrees=45
  swing2Motion=Free
  twistLimitDegrees=45
  twistMotion=Free
 }
 {
  bone1Name=dyn_coat_bk_E
  bone2Name=dyn_coat_c_E_r
  swing1LimitDegrees=45
  swing1Motion=Free
  swing2LimitDegrees=45
  swing2Motion=Free
  twistLimitDegrees=45
  twistMotion=Free
 }
 {
  bone1Name=foot_r
  bone2Name=calf_r
  swing1LimitDegrees=45
  swing1Motion=Free
  swing2LimitDegrees=45
  swing2Motion=Free
  twistLimitDegrees=45
  twistMotion=Free
 }
 {
  bone1Name=foot_l
  bone2Name=calf_l
  swing1LimitDegrees=45
  swing1Motion=Free
  swing2LimitDegrees=45
  swing2Motion=Free
  twistLimitDegrees=45
  twistMotion=Free
 }
 {
  bone1Name=pelvis
  bone2Name=root
  swing1LimitDegrees=45
  swing1Motion=Free
  swing2LimitDegrees=45
  swing2Motion=Free
  twistLimitDegrees=45
  twistMotion=Free
 }
 {
  bone1Name=dyn_coat_a_A_base_l
  bone2Name=root
  swing1LimitDegrees=45
  swing1Motion=Free
  swing2LimitDegrees=45
  swing2Motion=Free
  twistLimitDegrees=45
  twistMotion=Free
 }
 {
  bone1Name=dyn_coat_b_A_base_l
  bone2Name=root
  swing1LimitDegrees=45
  swing1Motion=Free
  swing2LimitDegrees=45
  swing2Motion=Free
  twistLimitDegrees=45
  twistMotion=Free
 }
 {
  bone1Name=dyn_coat_c_A_base_l
  bone2Name=root
  swing1LimitDegrees=45
  swing1Motion=Free
  swing2LimitDegrees=45
  swing2Motion=Free
  twistLimitDegrees=45
  twistMotion=Free
 }
 {
  bone1Name=dyn_coat_bk_A_base
  bone2Name=root
  swing1LimitDegrees=45
  swing1Motion=Free
  swing2LimitDegrees=45
  swing2Motion=Free
  twistLimitDegrees=45
  twistMotion=Free
 }
 {
  bone1Name=dyn_coat_b_A_base_r
  bone2Name=root
  swing1LimitDegrees=45
  swing1Motion=Free
  swing2LimitDegrees=45
  swing2Motion=Free
  twistLimitDegrees=45
  twistMotion=Free
 }
 {
  bone1Name=dyn_coat_c_A_base_r
  bone2Name=root
  swing1LimitDegrees=45
  swing1Motion=Free
  swing2LimitDegrees=45
  swing2Motion=Free
  twistLimitDegrees=45
  twistMotion=Free
 }
 {
  bone1Name=dyn_coat_a_A_base_r
  bone2Name=root
  swing1LimitDegrees=45
  swing1Motion=Free
  swing2LimitDegrees=45
  swing2Motion=Free
  twistLimitDegrees=45
  twistMotion=Free
 }
]
