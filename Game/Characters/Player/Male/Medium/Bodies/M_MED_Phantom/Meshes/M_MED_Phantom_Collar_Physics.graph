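# Physics asset for the collar mesh of the male medium Phantom player body, defining simulated collar panels with head and upper-spine bodies.
# Back and side panels are boxes, front panels and head are capsules, and each panel hangs off a tiny sphere base bone.
# Bases and head link to spine_05 with free rotation.
# Panel-to-base joints carry tight twist limits, with narrower swing on the front panels.
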PhysicsAsset /Game/Characters/Player/Male/Medium/Bodies/M_MED_Phantom/Meshes/M_MED_Phantom_Collar_Physics
bodies {
 dyn_collar_back [{shapeType=Box, extentX=17.5, extentY=9.0263, extentZ=19.5, center=<2.6564 0.0262 0>, rotation=<rot 0 -3.8846 0>}]
 dyn_collar_back_base [{shapeType=Sphere, radius=0.505}]
 dyn_collar_front_base_l [{shapeType=Sphere, radius=0.505}]
 dyn_collar_front_base_r [{shapeType=Sphere, radius=0.505}]
 dyn_collar_front_l [{shapeType=Capsule, length=10, radius=5, shapeName=dyn_collar_front_l_capsule, center=<4 0 0>, rotation=<rot 0 64.8712 -72.1948>}]
 dyn_collar_front_r [{shapeType=Capsule, length=10, radius=5, shapeName=dyn_collar_front_r_capsule, center=<-4 0 0>, rotation=<rot 0 65.0742 -74.3098>}]
 dyn_collar_side_base_l [{shapeType=Sphere, radius=0.505}]
 dyn_collar_side_base_r [{shapeType=Sphere, radius=0.505}]
 dyn_collar_side_l [{shapeType=Box, extentX=16, extentY=7.5044, extentZ=14.5, center=<3.5 -0.3011 -0.2425>, rotation=<rot -6.7695 -5.4545 7.4055>}]
 dyn_collar_side_r [{shapeType=Box, extentX=16, extentY=7.5044, extentZ=14.5, center=<-3.5 0.3011 0.2425>, rotation=<rot -6.7695 -5.4545 -172.5945>}]
 head [{shapeType=Capsule, length=5, radius=6, center=<4.0707 1.9298 -0.0001>, rotation=<rot 90 -0.3581 1.6763>}]
 spine_05 [{shapeType=Capsule, length=2.5, radius=2.75}]
}
constraints [
 {
  bone1Name=dyn_collar_front_base_l
  bone2Name=spine_05
  swing1LimitDegrees=45
  swing1Motion=Free
  swing2LimitDegrees=45
  swing2Motion=Free
  twistLimitDegrees=45
  twistMotion=Free
 }
 {
  bone1Name=dyn_collar_side_base_l
  bone2Name=spine_05
  swing1LimitDegrees=45
  swing1Motion=Free
  swing2LimitDegrees=45
  swing2Motion=Free
  twistLimitDegrees=45
  twistMotion=Free
 }
 {
  bone1Name=dyn_collar_back_base
  bone2Name=spine_05
  swing1LimitDegrees=45
  swing1Motion=Free
  swing2LimitDegrees=45
  swing2Motion=Free
  twistLimitDegrees=45
  twistMotion=Free
 }
 {
  bone1Name=dyn_collar_front_base_r
  bone2Name=spine_05
  swing1LimitDegrees=45
  swing1Motion=Free
  swing2LimitDegrees=45
  swing2Motion=Free
  twistLimitDegrees=45
  twistMotion=Free
 }
 {
  bone1Name=dyn_collar_side_base_r
  bone2Name=spine_05
  swing1LimitDegrees=45
  swing1Motion=Free
  swing2LimitDegrees=45
  swing2Motion=Free
  twistLimitDegrees=45
  twistMotion=Free
 }
 {bone1Name=dyn_collar_front_l, bone2Name=dyn_collar_front_base_l, swing1LimitDegrees=20, swing2LimitDegrees=45, twistLimitDegrees=5}
 {bone1Name=dyn_collar_side_l, bone2Name=dyn_collar_side_base_l, swing1LimitDegrees=45, swing2LimitDegrees=45, twistLimitDegrees=5}
 {bone1Name=dyn_collar_back, bone2Name=dyn_collar_back_base, swing1LimitDegrees=45, swing2LimitDegrees=45, twistLimitDegrees=5}
 {bone1Name=dyn_collar_side_r, bone2Name=dyn_collar_side_base_r, swing1LimitDegrees=45, swing2LimitDegrees=45, twistLimitDegrees=5}
 {bone1Name=dyn_collar_front_r, bone2Name=dyn_collar_front_base_r, swing1LimitDegrees=20, swing2LimitDegrees=45, twistLimitDegrees=5}
 {
  bone1Name=head
  bone2Name=spine_05
  swing1LimitDegrees=45
  swing1Motion=Free
  swing2LimitDegrees=45
  swing2Motion=Free
  twistLimitDegrees=45
  twistMotion=Free
 }
]
