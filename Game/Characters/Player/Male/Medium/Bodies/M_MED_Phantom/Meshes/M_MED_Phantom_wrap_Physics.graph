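# Physics asset for the M_MED_Phantom_wrap mesh on the male medium Phantom player body, giving its dynamic strap bones capsule bodies and joints.
# The left side forms a four-body chain from main base through main, A and B straps, while the right has only base and main.
# Constraints set swing and twist limits on each link, with the right joint held to half the range of the left ones.
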PhysicsAsset /Game/Characters/Player/Male/Medium/Bodies/M_MED_Phantom/Meshes/M_MED_Phantom_wrap_Physics
bodies {
 dyn_strap_a_l [{shapeType=Capsule, length=5.6609, radius=1.6866, center=<4.0006 0 0>, rotation=<rot 0 84.3325 -84.7272>}]
 dyn_strap_b_l [{shapeType=Capsule, length=6.4893, radius=1.7648, center=<4.8976 0 0>, rotation=<rot 0 90.3423 -86.8014>}]
 dyn_strap_main_base_l [{shapeType=Capsule, length=1, radius=1}]
 dyn_strap_main_base_r [{shapeType=Capsule, length=1, radius=1}]
 dyn_strap_main_l [{shapeType=Capsule, length=5.25, radius=4.75, center=<3.7711 0 0>, rotation=<rot 90 0 0>}]
 dyn_strap_main_r [{shapeType=Capsule, length=5.25, radius=4.75, center=<-2.8948 0.0001 0>, rotation=<rot 90 0 -180>}]
}
constraints [
 {bone1Name=dyn_strap_main_l, bone2Name=dyn_strap_main_base_l, swing1LimitDegrees=20, swing2LimitDegrees=20, twistLimitDegrees=10}
 {bone1Name=dyn_strap_main_r, bone2Name=dyn_strap_main_base_r, swing1LimitDegrees=10, swing2LimitDegrees=10, twistLimitDegrees=5}
 {bone1Name=dyn_strap_a_l, bone2Name=dyn_strap_main_l, swing1LimitDegrees=20, swing2LimitDegrees=20, twistLimitDegrees=10}
 {bone1Name=dyn_strap_b_l, bone2Name=dyn_strap_a_l, swing1LimitDegrees=20, swing2LimitDegrees=20, twistLimitDegrees=10}
]
