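# Phantom male body-part AnimBlueprint that animates coat, collar, wrap and straps over a pose copied from the attached parent mesh.
# Bone corrections driven by thigh, arm and head angles, crouch and slide surround three RigidBody simulations, while Perlin noise wiggles straps.
# When rigid bodies are disabled or LOD exceeds 1, the coat instead follows leg bones through constraints.
# Coat gravity uses a parameter-asset override, and skydiving or parachute events reset dynamics.
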AnimBlueprint /Game/Characters/Player/Male/Medium/Bodies/M_MED_Phantom/Meshes/M_MED_Phantom_AnimBP
target_skeleton=&/M_MED_Phantom_Skeleton
graphs [
 AnimGraph {nodes [
   n1 = "Constraint - dyn_coat_a_A_base_l (R)"(ComponentPose=Local To Component(LocalPose=Use cached pose 'SavedPose_main'()), Alpha=0.5, ConstraintWeights_0=1) {
    alphaBoolBlend {blendOption=Linear}
    boneToModify {boneName=dyn_coat_a_A_base_l}
    constraintSetup [{transformType=Rotation, targetBone {boneName=thigh_twist_01_l}}]
    constraintWeights [1]
   }
   n2 = "Constraint - dyn_coat_a_A_base_r (R)"(ComponentPose=@n1, Alpha=0.5, ConstraintWeights_0=1) {
    alphaBoolBlend {blendOption=Linear}
    boneToModify {boneName=dyn_coat_a_A_base_r}
    constraintSetup [{transformType=Rotation, targetBone {boneName=thigh_twist_01_r}}]
    constraintWeights [1]
   }
   n3 = "Constraint - dyn_coat_bk_A_base (PP)"!(ComponentPose=@n2, Alpha=1, ConstraintWeights_0=0.5, ConstraintWeights_1:float=0.5) -> (Pose) {
    alphaBoolBlend {blendOption=Linear}
    boneToModify {boneName=dyn_coat_bk_A_base}
    constraintSetup [{targetBone {boneName=dyn_coat_c_A_base_l}}, {targetBone {boneName=dyn_coat_c_A_base_r}}]
    constraintWeights [1,1]
   }
   n4 = "Transform (Modify) Bone - Bone: dyn_coat_bk_A_base"(ComponentPose=@n3, Alpha=1, Translation="0,0,0", Rotation="10,0,0", Scale="1,1,1") {
    rotationMode=BMM_Additive
    rotationSpace=BCS_BoneSpace
    scaleMode=BMM_Additive
    scaleSpace=BCS_BoneSpace
    translationMode=BMM_Additive
    translationSpace=BCS_BoneSpace
    alphaBoolBlend {blendOption=Linear}
    alphaScaleBiasClamp {inRange {max=0, min=-100}, outRange {max=0, min=1}}
    boneToModify {boneName=dyn_coat_bk_A_base}
    rotation {pitch=10}
   }
   n5 = "Transform (Modify) Bone - Bone: dyn_coat_a_A_base_l"(ComponentPose=@n4, Alpha=1, Translation="0,0,2", Rotation="0,0,0", Scale="1,1,1") {
    rotationMode=BMM_Additive
    rotationSpace=BCS_BoneSpace
    scaleMode=BMM_Additive
    scaleSpace=BCS_BoneSpace
    translationMode=BMM_Additive
    translationSpace=BCS_BoneSpace
    alphaBoolBlend {blendOption=Linear}
    alphaScaleBiasClamp {inRange {max=0, min=-100}, outRange {max=0, min=1}}
    boneToModify {boneName=dyn_coat_a_A_base_l}
    translation {z=2}
   }
   n6 = "Transform (Modify) Bone - Bone: dyn_coat_b_A_base_l"(ComponentPose=@n5, Alpha=1, Translation="0,0,2", Rotation="0,0,0", Scale="1,1,1") {
    rotationMode=BMM_Additive
    rotationSpace=BCS_BoneSpace
    scaleMode=BMM_Additive
    scaleSpace=BCS_BoneSpace
    translationMode=BMM_Additive
    translationSpace=BCS_BoneSpace
    alphaBoolBlend {blendOption=Linear}
    alphaScaleBiasClamp {inRange {max=0, min=-100}, outRange {max=0, min=1}}
    boneToModify {boneName=dyn_coat_b_A_base_l}
    translation {z=2}
   }
   n7 = "Transform (Modify) Bone - Bone: dyn_coat_a_A_base_r"(ComponentPose=@n6, Alpha=1, Translation="0,0,-2", Rotation="0,0,0", Scale="1,1,1") {
    rotationMode=BMM_Additive
    rotationSpace=BCS_BoneSpace
    scaleMode=BMM_Additive
    scaleSpace=BCS_BoneSpace
    translationMode=BMM_Additive
    translationSpace=BCS_BoneSpace
    alphaBoolBlend {blendOption=Linear}
    alphaScaleBiasClamp {inRange {max=0, min=-100}, outRange {max=0, min=1}}
    boneToModify {boneName=dyn_coat_a_A_base_r}
    translation {z=-2}
   }
   n8 = "Transform (Modify) Bone - Bone: dyn_coat_b_A_base_r"(ComponentPose=@n7, Alpha=1, Translation="0,0,-2", Rotation="0,0,0", Scale="1,1,1") {
    rotationMode=BMM_Additive
    rotationSpace=BCS_BoneSpace
    scaleMode=BMM_Additive
    scaleSpace=BCS_BoneSpace
    translationMode=BMM_Additive
    translationSpace=BCS_BoneSpace
    alphaBoolBlend {blendOption=Linear}
    alphaScaleBiasClamp {inRange {max=0, min=-100}, outRange {max=0, min=1}}
    boneToModify {boneName=dyn_coat_b_A_base_r}
    translation {z=-2}
   }
   n9 = Get legUp_L() -> (legUp_L:double)
   n10 = "Transform (Modify) Bone - Bone: dyn_coat_a_A_base_l"(ComponentPose=@n8, Alpha=@n9, Translation="-3,-13,6", Rotation="-50,-50,0", Scale="1,1,1") {
    rotationMode=BMM_Additive
    rotationSpace=BCS_BoneSpace
    scaleMode=BMM_Additive
    scaleSpace=BCS_BoneSpace
    translationMode=BMM_Additive
    translationSpace=BCS_BoneSpace
    alphaBoolBlend {blendOption=Linear}
    alphaScaleBiasClamp {bMapRange=true, inRange {max=0, min=-100}, outRange {max=0, min=1}}
    boneToModify {boneName=dyn_coat_a_A_base_l}
    rotation {pitch=-50, yaw=-50}
    translation=<-3 -13 6>
   }
   n11 = "Transform (Modify) Bone - Bone: dyn_coat_b_A_base_l"(ComponentPose=@n10, Alpha=@n9, Translation="-1.000,-5,1", Rotation="-15,-15,0", Scale="1,1,1") {
    rotationMode=BMM_Additive
    rotationSpace=BCS_BoneSpace
    scaleMode=BMM_Additive
    scaleSpace=BCS_BoneSpace
    translationMode=BMM_Additive
    translationSpace=BCS_BoneSpace
    alphaBoolBlend {blendOption=Linear}
    alphaScaleBiasClamp {bMapRange=true, inRange {max=0, min=-100}, outRange {max=0, min=1}}
    boneToModify {boneName=dyn_coat_b_A_base_l}
    rotation {pitch=-15, yaw=-15}
    translation=<-1 -5 1>
   }
   n12 = "Transform (Modify) Bone - Bone: dyn_coat_c_A_base_l"(ComponentPose=@n11, Alpha=@n9, Translation="-1.000,0,0", Rotation="15,-15,0", Scale="1,1,1") {
    rotationMode=BMM_Additive
    rotationSpace=BCS_BoneSpace
    scaleMode=BMM_Additive
    scaleSpace=BCS_BoneSpace
    translationMode=BMM_Additive
    translationSpace=BCS_BoneSpace
    alphaBoolBlend {blendOption=Linear}
    alphaScaleBiasClamp {bMapRange=true, inRange {max=0, min=-100}, outRange {max=0, min=1}}
    boneToModify {boneName=dyn_coat_c_A_base_l}
    rotation {pitch=15, yaw=-15}
    translation {x=-1}
   }
   n13 = Get legUp_R() -> (legUp_R:double)
   n14 = "Transform (Modify) Bone - Bone: dyn_coat_a_A_base_r"(ComponentPose=@n12, Alpha=@n13, Translation="3,13,-6", Rotation="-50,-50,0", Scale="1,1,1") {
    rotationMode=BMM_Additive
    rotationSpace=BCS_BoneSpace
    scaleMode=BMM_Additive
    scaleSpace=BCS_BoneSpace
    translationMode=BMM_Additive
    translationSpace=BCS_BoneSpace
    alphaBoolBlend {blendOption=Linear}
    alphaScaleBiasClamp {bMapRange=true, inRange {max=0, min=-100}, outRange {max=0, min=1}}
    boneToModify {boneName=dyn_coat_a_A_base_r}
    rotation {pitch=-50, yaw=-50}
    translation=<3 13 -6>
   }
   n15 = "Transform (Modify) Bone - Bone: dyn_coat_b_A_base_r"(ComponentPose=@n14, Alpha=@n13, Translation="1,5,-1", Rotation="-15,-15,0", Scale="1,1,1") {
    rotationMode=BMM_Additive
    rotationSpace=BCS_BoneSpace
    scaleMode=BMM_Additive
    scaleSpace=BCS_BoneSpace
    translationMode=BMM_Additive
    translationSpace=BCS_BoneSpace
    alphaBoolBlend {blendOption=Linear}
    alphaScaleBiasClamp {bMapRange=true, inRange {max=0, min=-100}, outRange {max=0, min=1}}
    boneToModify {boneName=dyn_coat_b_A_base_r}
    rotation {pitch=-15, yaw=-15}
    translation=<1 5 -1>
   }
   n16 = "Transform (Modify) Bone - Bone: dyn_coat_c_A_base_r"(ComponentPose=@n15, Alpha=@n13, Translation="1,0,0", Rotation="15,-15,0", Scale="1,1,1") {
    rotationMode=BMM_Additive
    rotationSpace=BCS_BoneSpace
    scaleMode=BMM_Additive
    scaleSpace=BCS_BoneSpace
    translationMode=BMM_Additive
    translationSpace=BCS_BoneSpace
    alphaBoolBlend {blendOption=Linear}
    alphaScaleBiasClamp {bMapRange=true, inRange {max=0, min=-100}, outRange {max=0, min=1}}
    boneToModify {boneName=dyn_coat_c_A_base_r}
    rotation {pitch=15, yaw=-15}
    translation {x=1}
   }
   n17 = Get legSide_L() -> (legSide_L:double)
   n18 = "Transform (Modify) Bone - Bone: dyn_coat_a_A_base_l"(ComponentPose=@n16, Alpha=@n17, Translation="0,0,0", Rotation="25,0,0", Scale="1,1,1") {
    rotationMode=BMM_Additive
    rotationSpace=BCS_BoneSpace
    scaleMode=BMM_Additive
    scaleSpace=BCS_BoneSpace
    translationMode=BMM_Additive
    translationSpace=BCS_BoneSpace
    alphaBoolBlend {blendOption=Linear}
    alphaScaleBiasClamp {bMapRange=true, inRange {max=20}}
    boneToModify {boneName=dyn_coat_a_A_base_l}
    rotation {pitch=25}
   }
   n19 = "Transform (Modify) Bone - Bone: dyn_coat_b_A_base_l"(ComponentPose=@n18, Alpha=@n17, Translation="0,0,0", Rotation="15,0,0", Scale="1,1,1") {
    rotationMode=BMM_Additive
    rotationSpace=BCS_BoneSpace
    scaleMode=BMM_Additive
    scaleSpace=BCS_BoneSpace
    translationMode=BMM_Additive
    translationSpace=BCS_BoneSpace
    alphaBoolBlend {blendOption=Linear}
    alphaScaleBiasClamp {bMapRange=true, inRange {max=20}}
    boneToModify {boneName=dyn_coat_b_A_base_l}
    rotation {pitch=15}
   }
   n20 = Get legSide_R() -> (legSide_R:double)
   n21 = "Transform (Modify) Bone - Bone: dyn_coat_a_A_base_r"(ComponentPose=@n19, Alpha=@n20, Translation="0,0,0", Rotation="25,0,0", Scale="1,1,1") {
    rotationMode=BMM_Additive
    rotationSpace=BCS_BoneSpace
    scaleMode=BMM_Additive
    scaleSpace=BCS_BoneSpace
    translationMode=BMM_Additive
    translationSpace=BCS_BoneSpace
    alphaBoolBlend {blendOption=Linear}
    alphaScaleBiasClamp {bMapRange=true, inRange {max=20}}
    boneToModify {boneName=dyn_coat_a_A_base_r}
    rotation {pitch=25}
   }
   n22 = "Transform (Modify) Bone - Bone: dyn_coat_b_A_base_r"(ComponentPose=@n21, Alpha=@n20, Translation="0,0,0", Rotation="25,0,0", Scale="1,1,1") {
    rotationMode=BMM_Additive
    rotationSpace=BCS_BoneSpace
    scaleMode=BMM_Additive
    scaleSpace=BCS_BoneSpace
    translationMode=BMM_Additive
    translationSpace=BCS_BoneSpace
    alphaBoolBlend {blendOption=Linear}
    alphaScaleBiasClamp {bMapRange=true, inRange {max=20}}
    boneToModify {boneName=dyn_coat_b_A_base_r}
    rotation {pitch=25}
   }
   n23 = Get crouchStill() -> (crouchStill:double)
   n24 = "Transform (Modify) Bone - Bone: dyn_coat_a_A_base_r"(ComponentPose=@n22, Alpha=@n23, Translation="0,0,0", Rotation="-15,-20,0", Scale="1,1,1") {
    rotationMode=BMM_Additive
    rotationSpace=BCS_BoneSpace
    scaleMode=BMM_Additive
    scaleSpace=BCS_BoneSpace
    translationMode=BMM_Additive
    translationSpace=BCS_BoneSpace
    alphaBoolBlend {blendOption=Linear}
    alphaScaleBiasClamp {inRange {max=0, min=-100}, outRange {max=0, min=1}}
    boneToModify {boneName=dyn_coat_a_A_base_r}
    rotation {pitch=-15, yaw=-20}
   }
   n25 = "Transform (Modify) Bone - Bone: dyn_coat_b_A_base_r"(ComponentPose=@n24, Alpha=@n23, Translation="0,0,0", Rotation="5,-30,0", Scale="1,1,1") {
    rotationMode=BMM_Additive
    rotationSpace=BCS_BoneSpace
    scaleMode=BMM_Additive
    scaleSpace=BCS_BoneSpace
    translationMode=BMM_Additive
    translationSpace=BCS_BoneSpace
    alphaBoolBlend {blendOption=Linear}
    alphaScaleBiasClamp {inRange {max=0, min=-100}, outRange {max=0, min=1}}
    boneToModify {boneName=dyn_coat_b_A_base_r}
    rotation {pitch=5, yaw=-30}
   }
   n26 = "Transform (Modify) Bone - Bone: dyn_coat_c_A_base_r"(ComponentPose=@n25, Alpha=@n23, Translation="0,0,0", Rotation="40,-15,0", Scale="1,1,1") {
    rotationMode=BMM_Additive
    rotationSpace=BCS_BoneSpace
    scaleMode=BMM_Additive
    scaleSpace=BCS_BoneSpace
    translationMode=BMM_Additive
    translationSpace=BCS_BoneSpace
    alphaBoolBlend {blendOption=Linear}
    alphaScaleBiasClamp {inRange {max=0, min=-100}, outRange {max=0, min=1}}
    boneToModify {boneName=dyn_coat_c_A_base_r}
    rotation {pitch=40, yaw=-15}
   }
   n27 = "Transform (Modify) Bone - Bone: dyn_coat_bk_A_base"(ComponentPose=@n26, Alpha=@n23, Translation="0,0,0", Rotation="15,0,0", Scale="1,1,1") {
    rotationMode=BMM_Additive
    rotationSpace=BCS_BoneSpace
    scaleMode=BMM_Additive
    scaleSpace=BCS_BoneSpace
    translationMode=BMM_Additive
    translationSpace=BCS_BoneSpace
    alphaBoolBlend {blendOption=Linear}
    alphaScaleBiasClamp {inRange {max=0, min=-100}, outRange {max=0, min=1}}
    boneToModify {boneName=dyn_coat_bk_A_base}
    rotation {pitch=15}
   }
   n28 = "Transform (Modify) Bone - Bone: dyn_coat_c_A_base_l"(ComponentPose=@n27, Alpha=0, Translation="0,0,0", Rotation="10,-5,0", Scale="1,1,1") {
    rotationMode=BMM_Additive
    rotationSpace=BCS_BoneSpace
    scaleMode=BMM_Additive
    scaleSpace=BCS_BoneSpace
    translationMode=BMM_Additive
    translationSpace=BCS_BoneSpace
    alphaBoolBlend {blendOption=Linear}
    alphaScaleBiasClamp {inRange {max=0, min=-100}, outRange {max=0, min=1}}
    boneToModify {boneName=dyn_coat_c_A_base_l}
    rotation {pitch=10, yaw=-5}
   }
   n29 = RigidBody!(ComponentPose=@n28, Alpha, OverrideWorldGravity:Vector=Get Gravity Override(), ExternalForce) -> (Pose) {
    bClampLinearTranslationLimitToRefPose=true
    bOverrideWorldGravity=true
    overridePhysicsAsset=&/M_MED_Phantom_Coat_Physics
    simulationSpace=BaseBoneSpace
    alphaBoolBlend {blendOption=Linear}
    baseBoneRef {boneName=root}
    componentAppliedLinearAccClamp=<2000 2000 2000>
    componentLinearAccScale=<0.4 0.1 0.7>
    componentLinearVelScale=<0.4 0.1 0.7>
    overrideWorldGravity {z=-980}
   }
   n30 = RigidBody!(ComponentPose=@n29, Alpha, ExternalForce) -> (Pose) {
    bClampLinearTranslationLimitToRefPose=true
    bOverrideWorldGravity=true
    overridePhysicsAsset=&/M_MED_Phantom_Collar_Physics
    simulationSpace=BaseBoneSpace
    alphaBoolBlend {blendOption=Linear}
    baseBoneRef {boneName=spine_04}
    componentAppliedLinearAccClamp=<2000 2000 2000>
    componentLinearAccScale=<0.5 0.5 0.5>
    componentLinearVelScale=<0.5 0.5 0.5>
    overrideWorldGravity {z=-980}
   }
   n31 = RigidBody(ComponentPose=@n30) {
    bClampLinearTranslationLimitToRefPose=true
    bOverrideWorldGravity=true
    overridePhysicsAsset=&/M_MED_Phantom_wrap_Physics
    simulationSpace=BaseBoneSpace
    alphaBoolBlend {blendOption=Linear}
    baseBoneRef {boneName=spine_05}
    componentAppliedLinearAccClamp=<2000 2000 2000>
    componentLinearAccScale=<0.1 0.1 0.1>
    componentLinearVelScale=<0.1 0.1 0.1>
    overrideWorldGravity {z=-100}
   }
   n32 = Get strapTran_L() -> (strapTran_L:Vector)
   n33 = "Transform (Modify) Bone - Bone: dyn_strap_main_l"(ComponentPose=@n31, Alpha=1, Translation=@n32, Rotation="0,0,0", Scale="1,1,1") {
    rotationMode=BMM_Additive
    rotationSpace=BCS_BoneSpace
    scaleMode=BMM_Additive
    scaleSpace=BCS_BoneSpace
    translationMode=BMM_Additive
    translationSpace=BCS_BoneSpace
    alphaBoolBlend {blendOption=Linear}
    alphaScaleBiasClamp {inRange {max=0, min=-100}, outRange {max=0, min=1}}
    boneToModify {boneName=dyn_strap_main_l}
   }
   n34 = "Transform (Modify) Bone - Bone: dyn_strap_main_r"(ComponentPose=@n33, Alpha=0.7, Translation=@n32, Rotation="0,0,0", Scale="1,1,1") {
    rotationMode=BMM_Additive
    rotationSpace=BCS_BoneSpace
    scaleMode=BMM_Additive
    scaleSpace=BCS_BoneSpace
    translationMode=BMM_Additive
    translationSpace=BCS_BoneSpace
    alphaBoolBlend {blendOption=Linear}
    alphaScaleBiasClamp {inRange {max=0, min=-100}, outRange {max=0, min=1}}
    boneToModify {boneName=dyn_strap_main_r}
   }
   n35 = "Transform (Modify) Bone - Bone: dyn_strap_a_l"(ComponentPose=@n34, Alpha=1, Translation="0,0,0", Rotation=Get strapRotA() -> (strapRotA:Rotator), Scale="1,1,1") {
    rotationMode=BMM_Additive
    rotationSpace=BCS_BoneSpace
    scaleMode=BMM_Additive
    scaleSpace=BCS_BoneSpace
    translationMode=BMM_Additive
    translationSpace=BCS_BoneSpace
    alphaBoolBlend {blendOption=Linear}
    alphaScaleBiasClamp {inRange {max=0, min=-100}, outRange {max=0, min=1}}
    boneToModify {boneName=dyn_strap_a_l}
   }
   n36 = "Transform (Modify) Bone - Bone: dyn_strap_b_l"(ComponentPose=@n35, Alpha=1, Translation="0,0,0", Rotation=Get strapRotB() -> (strapRotB:Rotator), Scale="1,1,1") {
    rotationMode=BMM_Additive
    rotationSpace=BCS_BoneSpace
    scaleMode=BMM_Additive
    scaleSpace=BCS_BoneSpace
    translationMode=BMM_Additive
    translationSpace=BCS_BoneSpace
    alphaBoolBlend {blendOption=Linear}
    alphaScaleBiasClamp {inRange {max=0, min=-100}, outRange {max=0, min=1}}
    boneToModify {boneName=dyn_strap_b_l}
   }
   n37 = Get strapRotC() -> (strapRotC:Rotator)
   n38 = "Transform (Modify) Bone - Bone: dyn_strap_a_r"(ComponentPose=@n36, Alpha=0.2, Translation="0,0,0", Rotation=@n37, Scale="1,1,1") {
    rotationMode=BMM_Additive
    rotationSpace=BCS_BoneSpace
    scaleMode=BMM_Additive
    scaleSpace=BCS_BoneSpace
    translationMode=BMM_Additive
    translationSpace=BCS_BoneSpace
    alphaBoolBlend {blendOption=Linear}
    alphaScaleBiasClamp {inRange {max=0, min=-100}, outRange {max=0, min=1}}
    boneToModify {boneName=dyn_strap_a_r}
   }
   n39 = "Transform (Modify) Bone - Bone: dyn_strap_b_r"(ComponentPose=@n38, Alpha=1, Translation="0,0,0", Rotation=@n37, Scale="1,1,1") {
    rotationMode=BMM_Additive
    rotationSpace=BCS_BoneSpace
    scaleMode=BMM_Additive
    scaleSpace=BCS_BoneSpace
    translationMode=BMM_Additive
    translationSpace=BCS_BoneSpace
    alphaBoolBlend {blendOption=Linear}
    alphaScaleBiasClamp {inRange {max=0, min=-100}, outRange {max=0, min=1}}
    boneToModify {boneName=dyn_strap_b_r}
   }
   n40 = "Transform (Modify) Bone - Bone: dyn_strap_b_r"(ComponentPose=@n39, Alpha=Get strapAlpha_R() -> (strapAlpha_R:double), Translation="0,0,0", Rotation="0,25,0", Scale="1,1,1") {
    rotationMode=BMM_Additive
    rotationSpace=BCS_BoneSpace
    scaleMode=BMM_Additive
    scaleSpace=BCS_BoneSpace
    translationMode=BMM_Additive
    translationSpace=BCS_BoneSpace
    alphaBoolBlend {blendOption=Linear}
    alphaScaleBiasClamp {inRange {max=0, min=-100}, outRange {max=0, min=1}}
    boneToModify {boneName=dyn_strap_b_r}
    rotation {yaw=25}
   }
   n41 = "Constraint - dyn_collar_back (P)"(ComponentPose=@n40, Alpha=Get lookUpAlpha() -> (lookUpAlpha:double), ConstraintWeights_0=1) {
    alphaBoolBlend {blendOption=Linear}
    alphaScaleBiasClamp {scale=0.9}
    boneToModify {boneName=dyn_collar_back}
    constraintSetup [{targetBone {boneName=head}}]
    constraintWeights [1]
   }
   n42 = Get isSliding() -> (isSliding:double)
   n43 = "Transform (Modify) Bone - Bone: dyn_coat_bk_B"!(ComponentPose=@n41, Alpha=@n42, Translation="-0.000,0,18", Scale="1,1,1") -> (Pose) {
    rotationMode=BMM_Additive
    rotationSpace=BCS_BoneSpace
    scaleMode=BMM_Additive
    scaleSpace=BCS_BoneSpace
    translationMode=BMM_Additive
    translationSpace=BCS_BoneSpace
    alphaBoolBlend {blendOption=Linear}
    alphaScaleBiasClamp {bInterpResult=true, interpSpeedIncreasing=5, inRange {max=0, min=-100}, outRange {max=0, min=1}}
    boneToModify {boneName=dyn_coat_bk_B}
    translation {z=18}
   }
   n44 = "Transform (Modify) Bone - Bone: dyn_coat_c_B_r"!(ComponentPose=@n43, Alpha=@n42, Translation="0.000,0.000,-36", Scale="1,1,1") -> (Pose) {
    rotationMode=BMM_Additive
    rotationSpace=BCS_BoneSpace
    scaleMode=BMM_Additive
    scaleSpace=BCS_BoneSpace
    translationMode=BMM_Additive
    translationSpace=BCS_BoneSpace
    alphaBoolBlend {blendOption=Linear}
    alphaScaleBiasClamp {bInterpResult=true, interpSpeedIncreasing=5, inRange {max=0, min=-100}, outRange {max=0, min=1}}
    boneToModify {boneName=dyn_coat_c_B_r}
    translation {z=-36}
   }
   n45 = "Transform (Modify) Bone - Bone: dyn_coat_c_B_l"!(ComponentPose=@n44, Alpha=@n42, Translation="0,0,36", Scale="1,1,1") -> (Pose) {
    rotationMode=BMM_Additive
    rotationSpace=BCS_BoneSpace
    scaleMode=BMM_Additive
    scaleSpace=BCS_BoneSpace
    translationMode=BMM_Additive
    translationSpace=BCS_BoneSpace
    alphaBoolBlend {blendOption=Linear}
    alphaScaleBiasClamp {bInterpResult=true, interpSpeedIncreasing=5, inRange {max=0, min=-100}, outRange {max=0, min=1}}
    boneToModify {boneName=dyn_coat_c_B_l}
    translation {z=36}
   }
   n46 = "Transform (Modify) Bone - Bone: dyn_coat_bk_C"!(ComponentPose=@n45, Alpha=@n42, Translation="-0.000,0.000,-18", Scale="1,1,1") -> (Pose) {
    rotationMode=BMM_Additive
    rotationSpace=BCS_BoneSpace
    scaleMode=BMM_Additive
    scaleSpace=BCS_BoneSpace
    translationMode=BMM_Additive
    translationSpace=BCS_BoneSpace
    alphaBoolBlend {blendOption=Linear}
    alphaScaleBiasClamp {bInterpResult=true, interpSpeedIncreasing=5, inRange {max=0, min=-100}, outRange {max=0, min=1}}
    boneToModify {boneName=dyn_coat_bk_C}
    translation {z=-18}
   }
   n47 = "Transform (Modify) Bone - Bone: dyn_coat_c_C_r"!(ComponentPose=@n46, Alpha=@n42, Translation="0.000,0,20", Scale="1,1,1") -> (Pose) {
    rotationMode=BMM_Additive
    rotationSpace=BCS_BoneSpace
    scaleMode=BMM_Additive
    scaleSpace=BCS_BoneSpace
    translationMode=BMM_Additive
    translationSpace=BCS_BoneSpace
    alphaBoolBlend {blendOption=Linear}
    alphaScaleBiasClamp {bInterpResult=true, interpSpeedIncreasing=5, inRange {max=0, min=-100}, outRange {max=0, min=1}}
    boneToModify {boneName=dyn_coat_c_C_r}
    translation {z=20}
   }
   n48 = "Transform (Modify) Bone - Bone: dyn_coat_c_C_l"!(ComponentPose=@n47, Alpha=@n42, Translation="0,0,-20", Scale="1,1,1") -> (Pose) {
    rotationMode=BMM_Additive
    rotationSpace=BCS_BoneSpace
    scaleMode=BMM_Additive
    scaleSpace=BCS_BoneSpace
    translationMode=BMM_Additive
    translationSpace=BCS_BoneSpace
    alphaBoolBlend {blendOption=Linear}
    alphaScaleBiasClamp {bInterpResult=true, interpSpeedIncreasing=5, inRange {max=0, min=-100}, outRange {max=0, min=1}}
    boneToModify {boneName=dyn_coat_c_C_l}
    translation {z=-20}
   }
   n49 = "Transform (Modify) Bone - Bone: dyn_coat_bk_D"!(ComponentPose=@n48, Alpha=@n42, Translation="0.000,-4.6,13", Scale="1,1,1") -> (Pose) {
    rotationMode=BMM_Additive
    rotationSpace=BCS_BoneSpace
    scaleMode=BMM_Additive
    scaleSpace=BCS_BoneSpace
    translationMode=BMM_Additive
    translationSpace=BCS_BoneSpace
    alphaBoolBlend {blendOption=Linear}
    alphaScaleBiasClamp {bInterpResult=true, interpSpeedIncreasing=5, inRange {max=0, min=-100}, outRange {max=0, min=1}}
    boneToModify {boneName=dyn_coat_bk_D}
    rotation=<rot -35.7097 0 0>
    translation {y=-4.6, z=13}
   }
   n50 = "Transform (Modify) Bone - Bone: dyn_coat_c_D_r"!(ComponentPose=@n49, Alpha=@n42, Translation="0,-5,10", Scale="1,1,1") -> (Pose) {
    rotationMode=BMM_Additive
    rotationSpace=BCS_BoneSpace
    scaleMode=BMM_Additive
    scaleSpace=BCS_BoneSpace
    translationMode=BMM_Additive
    translationSpace=BCS_BoneSpace
    alphaBoolBlend {blendOption=Linear}
    alphaScaleBiasClamp {bInterpResult=true, interpSpeedIncreasing=5, inRange {max=0, min=-100}, outRange {max=0, min=1}}
    boneToModify {boneName=dyn_coat_c_D_r}
    translation {y=-5, z=10}
   }
   n51 = "Transform (Modify) Bone - Bone: dyn_coat_c_D_l"!(ComponentPose=@n50, Alpha=@n42, Translation="0,-5,-10", Scale="1,1,1") -> (Pose) {
    rotationMode=BMM_Additive
    rotationSpace=BCS_BoneSpace
    scaleMode=BMM_Additive
    scaleSpace=BCS_BoneSpace
    translationMode=BMM_Additive
    translationSpace=BCS_BoneSpace
    alphaBoolBlend {blendOption=Linear}
    alphaScaleBiasClamp {bInterpResult=true, interpSpeedIncreasing=5, inRange {max=0, min=-100}, outRange {max=0, min=1}}
    boneToModify {boneName=dyn_coat_c_D_l}
    translation {y=-5, z=-10}
   }
   n52 = "Constraint - dyn_coat_a_A_base_l (P)"(ComponentPose=Local To Component(LocalPose=Use cached pose 'SavedPose_main'()), Alpha=1, ConstraintWeights_0=1) {
    alphaBoolBlend {blendOption=Linear}
    boneToModify {boneName=dyn_coat_a_A_base_l}
    constraintSetup [{targetBone {boneName=thigh_twist_01_l}}]
    constraintWeights [1]
   }
   n53 = "Constraint - dyn_coat_a_D_l (P)"(ComponentPose=@n52, Alpha=1, ConstraintWeights_0=1) {
    alphaBoolBlend {blendOption=Linear}
    boneToModify {boneName=dyn_coat_a_D_l}
    constraintSetup [{targetBone {boneName=calf_l}}]
    constraintWeights [1]
   }
   n54 = "Constraint - dyn_coat_b_A_base_l (P)"(ComponentPose=@n53, Alpha=1, ConstraintWeights_0=1) {
    alphaBoolBlend {blendOption=Linear}
    boneToModify {boneName=dyn_coat_b_A_base_l}
    constraintSetup [{targetBone {boneName=thigh_twist_01_l}}]
    constraintWeights [1]
   }
   n55 = "Constraint - dyn_coat_b_D_l (P)"(ComponentPose=@n54, Alpha=1, ConstraintWeights_0=1) {
    alphaBoolBlend {blendOption=Linear}
    boneToModify {boneName=dyn_coat_b_D_l}
    constraintSetup [{targetBone {boneName=calf_l}}]
    constraintWeights [1]
   }
   n56 = "Constraint - dyn_coat_c_A_base_l (P)"(ComponentPose=@n55, Alpha=1, ConstraintWeights_0=1) {
    alphaBoolBlend {blendOption=Linear}
    boneToModify {boneName=dyn_coat_c_A_base_l}
    constraintSetup [{targetBone {boneName=thigh_twist_01_l}}]
    constraintWeights [1]
   }
   n57 = "Constraint - dyn_coat_c_D_l (P)"(ComponentPose=@n56, Alpha=1, ConstraintWeights_0=1) {
    alphaBoolBlend {blendOption=Linear}
    boneToModify {boneName=dyn_coat_c_D_l}
    constraintSetup [{targetBone {boneName=calf_l}}]
    constraintWeights [1]
   }
   n58 = "Constraint - dyn_coat_a_A_base_r (P)"(ComponentPose=@n57, Alpha=1, ConstraintWeights_0=1) {
    alphaBoolBlend {blendOption=Linear}
    boneToModify {boneName=dyn_coat_a_A_base_r}
    constraintSetup [{targetBone {boneName=thigh_twist_01_r}}]
    constraintWeights [1]
   }
   n59 = "Constraint - dyn_coat_a_D_r (P)"(ComponentPose=@n58, Alpha=1, ConstraintWeights_0=1) {
    alphaBoolBlend {blendOption=Linear}
    boneToModify {boneName=dyn_coat_a_D_r}
    constraintSetup [{targetBone {boneName=calf_r}}]
    constraintWeights [1]
   }
   n60 = "Constraint - dyn_coat_b_A_base_r (P)"(ComponentPose=@n59, Alpha=1, ConstraintWeights_0=1) {
    alphaBoolBlend {blendOption=Linear}
    boneToModify {boneName=dyn_coat_b_A_base_r}
    constraintSetup [{targetBone {boneName=thigh_twist_01_r}}]
    constraintWeights [1]
   }
   n61 = "Constraint - dyn_coat_b_D_r (P)"(ComponentPose=@n60, Alpha=1, ConstraintWeights_0=1) {
    alphaBoolBlend {blendOption=Linear}
    boneToModify {boneName=dyn_coat_b_D_r}
    constraintSetup [{targetBone {boneName=calf_r}}]
    constraintWeights [1]
   }
   n62 = "Constraint - dyn_coat_c_A_base_r (P)"(ComponentPose=@n61, Alpha=1, ConstraintWeights_0=1) {
    alphaBoolBlend {blendOption=Linear}
    boneToModify {boneName=dyn_coat_c_A_base_r}
    constraintSetup [{targetBone {boneName=thigh_twist_01_r}}]
    constraintWeights [1]
   }
   n63 = "Constraint - dyn_coat_c_D_r (P)"(ComponentPose=@n62, Alpha=1, ConstraintWeights_0=1) {
    alphaBoolBlend {blendOption=Linear}
    boneToModify {boneName=dyn_coat_c_D_r}
    constraintSetup [{targetBone {boneName=calf_r}}]
    constraintWeights [1]
   }
   n64 = "Constraint - dyn_coat_bk_A_base (PP)"(ComponentPose=@n63, Alpha=1, ConstraintWeights_0=0.5, ConstraintWeights_1=0.5) {
    alphaBoolBlend {blendOption=Linear}
    boneToModify {boneName=dyn_coat_bk_A_base}
    constraintSetup [{targetBone {boneName=dyn_coat_c_A_base_l}}, {targetBone {boneName=dyn_coat_c_A_base_r}}]
    constraintWeights [1,1]
   }
   n65 = "Constraint - dyn_coat_bk_D (PP)"!(ComponentPose=@n64, Alpha=1, ConstraintWeights_0=0.5, ConstraintWeights_1:float=0.5) -> (Pose) {
    alphaBoolBlend {blendOption=Linear}
    boneToModify {boneName=dyn_coat_bk_D}
    constraintSetup [{targetBone {boneName=dyn_coat_c_D_l}}, {targetBone {boneName=dyn_coat_c_D_r}}]
    constraintWeights [1,1]
   }
   Output Pose(Result=Blend Poses by bool(
     BlendPose_0=Component To Local(ComponentPose=@n65)
     BlendPose_1=Component To Local(ComponentPose=@n51)
     bActiveValue=Get isRigidBodyDisabled() -> (isRigidBodyDisabled:bool)
    ) {blendType=Linear}) {layerGroup=None}
   n66 = "Constraint - dyn_collar_front_base_l (P)"(ComponentPose=Local To Component(LocalPose=Copy Pose From Mesh() {bUseAttachedParent=true}), Alpha=1, ConstraintWeights_0=1) {
    alphaBoolBlend {blendOption=Linear}
    boneToModify {boneName=dyn_collar_front_base_l}
    constraintSetup [{targetBone {boneName=neck_01}}]
    constraintWeights [1]
   }
   n67 = "Constraint - dyn_collar_side_base_l (P)"(ComponentPose=@n66, Alpha=1, ConstraintWeights_0=1) {
    alphaBoolBlend {blendOption=Linear}
    boneToModify {boneName=dyn_collar_side_base_l}
    constraintSetup [{targetBone {boneName=neck_01}}]
    constraintWeights [1]
   }
   n68 = "Constraint - dyn_collar_back_base (P)"(ComponentPose=@n67, Alpha=1, ConstraintWeights_0=1) {
    alphaBoolBlend {blendOption=Linear}
    boneToModify {boneName=dyn_collar_back_base}
    constraintSetup [{targetBone {boneName=neck_01}}]
    constraintWeights [1]
   }
   n69 = "Constraint - dyn_collar_front_base_r (P)"(ComponentPose=@n68, Alpha=1, ConstraintWeights_0=1) {
    alphaBoolBlend {blendOption=Linear}
    boneToModify {boneName=dyn_collar_front_base_r}
    constraintSetup [{targetBone {boneName=neck_01}}]
    constraintWeights [1]
   }
   n70 = "Constraint - dyn_collar_side_base_r (P)"(ComponentPose=@n69, Alpha=1, ConstraintWeights_0=1) {
    alphaBoolBlend {blendOption=Linear}
    boneToModify {boneName=dyn_collar_side_base_r}
    constraintSetup [{targetBone {boneName=neck_01}}]
    constraintWeights [1]
   }
   n71 = "Transform (Modify) Bone - Bone: dyn_collar_front_base_r"(ComponentPose=@n70, Alpha=Get armUp_R() -> (armUp_R:double), Translation="0,0,0", Rotation="0,20,0", Scale="1,1,1") {
    rotationMode=BMM_Additive
    rotationSpace=BCS_BoneSpace
    scaleMode=BMM_Additive
    scaleSpace=BCS_BoneSpace
    translationMode=BMM_Additive
    translationSpace=BCS_BoneSpace
    alphaBoolBlend {blendOption=Linear}
    alphaScaleBiasClamp {bMapRange=true, inRange {max=20}}
    boneToModify {boneName=dyn_collar_front_base_r}
    rotation {yaw=20}
   }
   n72 = Look At - Bone: dyn_leg_collision_l(ComponentPose=@n71, Alpha=1) {
    alphaBoolBlend {blendOption=Linear}
    boneToModify {boneName=dyn_leg_collision_l}
    lookAtLocation {x=0}
    lookAtTarget {boneReference {boneName=foot_l}}
    lookAt_Axis {axis=<1 0>}
    lookUp_Axis {axis {x=0, z=1}}
   }
   n73 = Look At - Bone: dyn_reg_collision_r(ComponentPose=@n72, Alpha=1) {
    alphaBoolBlend {blendOption=Linear}
    boneToModify {boneName=dyn_reg_collision_r}
    lookAtLocation {x=0}
    lookAtTarget {boneReference {boneName=foot_r}}
    lookAt_Axis {axis=<-1 0>}
    lookUp_Axis {axis {x=0, z=1}}
   }
   Save cached pose 'SavedPose_main'(Pose=Component To Local(ComponentPose=@n73)) {cacheName=SavedPose_main}
   Break Fort Anim Input Skydiving External Force(FortAnimInput_SkydivingExternalForce=Get SkydivingExternalForceInput()) -> (
    bUseSkydivingVectorForce=~
    SkydivingVectorForce=~
    HeadToPelvisDirection=~
    FloatingMultiplier=~
    FloatingAdditive=~
    DivingMultiplier=~
    DivingAdditive=~
    ParachutingMultiplier=~
    ParachutingAdditive=~
   )
   Get jumpVelocity() -> (jumpVelocity:Vector)
  ]}
 EventGraph {nodes [
   Event BlueprintUpdateAnimation()
   > n1 = Sequence!(execute) -> (then_0, then_1, then_2:Exec, then_3:Exec, then_4:Exec, then_5:Exec, then_6:Exec, then_7:Exec, then_8:Exec)
   n2 = "Gravity Override State Function (Parameter Asset) (Nativized)"(
    execute=@n1.then_5
    CharacterPartAnimInstance=Self-Reference()
    GameTimeInSeconds=GetGameTimeInSeconds()
    Parameters=&/M_MED_Phantom_Parameters
   )
   > Set Gravity Override(Gravity Override=@n2)
   n3 = Branch(execute=@n1.then_6, Condition=Get bIsSkyDiving())
   > n4 = Branch(Condition=Get bIsParachuteOpened())
   Set gravity_override(execute=[@n4.then, @n3.else], gravity_override="0, 0,-980")
   Set gravity_override(execute=@n4.else, gravity_override="0, 0,100")
   n5 = Sequence!(execute=@n1.then_8) -> (then_0, then_1, then_2, then_3, then_4)
   > n6 = makeNoiseFNC!(execute, self, frequencey:double=0.3, amp:double=5) -> (then, output:double)
   > Set strapTran_L(strapTran_L:Vector=MakeVector(X=@n6, Y=0)) -> (Output_Get:Vector)
   n7 = makeNoiseFNC(execute=@n5.then_1, frequencey=0.3, amp=20)
   n8 = makeNoiseFNC(execute=@n5.then_2, frequencey=0.3, amp=20)
   n9 = makeNoiseFNC(execute=@n5.then_3, frequencey=0.3, amp=20)
   > Set strapRotA(strapRotA:Rotator=MakeRotator(Roll=@n7, Pitch=@n8, Yaw=@n9)) -> (Output_Get:Rotator)
   n10 = Sequence!(execute=@n5.then_4) -> (then_0, then_1, then_2, then_3)
   > n11 = makeNoiseFNC(frequencey=0.4, amp=20)
   n12 = makeNoiseFNC(execute=@n10.then_1, frequencey=0.4, amp=20)
   n13 = makeNoiseFNC(execute=@n10.then_2, frequencey=0.4, amp=20)
   > Set strapRotB(strapRotB:Rotator=MakeRotator(Roll=@n11, Pitch=@n12, Yaw=@n13)) -> (Output_Get:Rotator)
   n14 = Sequence!(execute=@n10.then_3) -> (then_0, then_1, then_2)
   > makeNoiseFNC(frequencey=0.4, amp=20)
   makeNoiseFNC(execute=@n14.then_1, frequencey=0.2, amp=5)
   n15 = makeNoiseFNC(execute=@n14.then_2, frequencey=0.2, amp=25)
   n16 = "Absolute (Float)"(A=@n15)
   > Set strapRotC(strapRotC:Rotator=MakeRotator(Yaw=float * float(A=@n16, B=-1))) -> (Output_Get:Rotator)
   n17 = GetOwnerSkeletalMeshComponent()
   n18 = BreakRotator(InRot=BreakTransform(InTransform=GetDeltaTransformFromRefPose(self=@n17, BoneName=thigh_l, BaseName=pelvis)).Rotation)
   Set legUp_L(execute=@n1.then_0, legUp_L:double=@n18.Yaw) -> (Output_Get:double)
   > Set legSide_L(legSide_L:double=@n18.Pitch) -> (Output_Get:double)
   n19 = BreakRotator(InRot=BreakTransform(InTransform=GetDeltaTransformFromRefPose(self=@n17, BoneName=thigh_r, BaseName=pelvis)).Rotation)
   Set legUp_R(execute=@n1.then_1, legUp_R:double=@n19.Yaw) -> (Output_Get:double)
   > Set legSide_R(legSide_R:double=@n19.Pitch) -> (Output_Get:double)
   n20 = OR Boolean(A=Get bIsJumping(), B=Get bIsSkyDiving())
   n21 = Get bIsFalling()
   Set coatSpreadAlpha(execute=@n1.then_7, coatSpreadAlpha:double="To Float (Boolean)"(InBool=OR Boolean(A=@n20, B=@n21))) -> (Output_Get:double)
   > n22 = Branch(Condition=Get bIsJumping())
   > Set jumpVelocity(jumpVelocity:Vector="5, 0,0") -> (Output_Get:Vector)
   Set jumpVelocity(execute=@n22.else, jumpVelocity="0, 0, 0")
   n23 = "Not Equal (Boolean)"!(A:bool=Get bIsMoving2D(), B:bool=true) -> (ReturnValue:bool) {is=title K2Node_CallFunction}
   n24 = Get bIsCrouching()
   Set crouchStill(execute=@n1.then_3, crouchStill:double="To Float (Boolean)"(InBool=AND Boolean(A=@n24, B=@n23))) -> (Output_Get:double)
   > Set strapAlpha_R(strapAlpha_R:double=Get Distance Between Two Sockets!(
     Component:Skeletal Mesh Component Object Reference=@n17
     SocketOrBoneNameA:Name=dyn_collar_front_r
     SocketSpaceA:ERelativeTransformSpace Enum=RTS_World
     SocketOrBoneNameB:Name=upperarm_twist_02_r
     SocketSpaceB:ERelativeTransformSpace Enum=RTS_World
     bRemapRange:bool=true
     InRangeMin:float=120
     InRangeMax:float=130
     OutRangeMin:float=0
     OutRangeMax:float=1
    ) -> (ReturnValue:float) {is=title K2Node_CallFunction}) -> (Output_Get:double)
   n25 = Get CurrentLOD()
   n26 = Get bRigidBodyEnabled()
   Set isRigidBodyDisabled(
    execute=@n1.then_4
    isRigidBodyDisabled:bool=OR Boolean(A=NOT Boolean(A=@n26), B="integer > integer"(A=@n25, B=1) {is=title K2Node_CallFunction})
   ) -> (Output_Get:bool)
   > Set isSliding(isSliding:double="To Float (Boolean)"(InBool=Get bIsSliding())) -> (Output_Get:double)
   n27 = BreakTransform(InTransform=GetDeltaTransformFromRefPose(self=GetOwnerSkeletalMeshComponent(), BoneName=head, BaseName=neck_01))
   > Set lookUpAlpha(lookUpAlpha:double=MapRangeClamped(Value=BreakRotator(InRot=@n27.Rotation).Yaw, InRangeA=-15, InRangeB=-25)) -> (Output_Get:double)
   n28 = GetDeltaTransformFromRefPose(self=@n17, BoneName=upperarm_r, BaseName=clavicle_r)
   Set armUp_R(execute=@n1.then_2, armUp_R:double=BreakRotator(InRot=BreakTransform(InTransform=@n28).Rotation).Pitch) -> (Output_Get:double)
   Event OnParachuteOpened()
   > n29 = Delay(Duration=1.5)
   ResetDynamics(execute=[@n29.then, @n30.then, @n31.then], InTeleportType=TeleportPhysics)
   > ResetDynamics(InTeleportType=ResetPhysics)
   n30 = Event OnBeginSkydiving()
   n31 = Event OnEndSkydiving()
  ]}
 makeNoiseFNC {nodes [
   n1 = makeNoiseFNC!() -> (then, frequencey:double=~, amp:double=~) {is=title K2Node_FunctionEntry}
   n2 = float * float(A=@n1, B=GetGameTimeInSeconds())
   n3 = float * float(A="Sin (Radians)"(A=PerlinNoise1D!(Value:float=@n2) -> (ReturnValue:float)), B=@n1.amp)
   > Return Node!(execute, output:double=@n3)
   PerlinNoise1D(Value=@n3)
  ]}
]
variables [
 armUp_R {type=number}
 coatSpreadAlpha {type=number}
 crouchStill {type=number}
 footCollisionScale {type=Vector}
 Gravity Override {type=Vector}
 Gravity Override Params {type=GravityOverrideParamsStruct}
 gravity_override {type=Vector}
 isRigidBodyDisabled {type=boolean}
 isSliding {type=number}
 jump_alpha {type=number}
 jumpVelocity {type=Vector}
 kneeCollisionScale {type=Vector}
 legSide_L {type=number}
 legSide_R {type=number}
 legUp_L {type=number}
 legUp_R {type=number}
 lookUpAlpha {type=number}
 movingFwd {type=number}
 strapAlpha_R {type=number}
 strapRotA {type=Rotator}
 strapRotB {type=Rotator}
 strapRotC {type=Rotator}
 strapTran_L {type=Vector}
 strapTran_R {type=Vector}
]
